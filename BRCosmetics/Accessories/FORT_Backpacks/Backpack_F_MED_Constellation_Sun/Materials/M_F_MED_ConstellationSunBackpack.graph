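# Translucent two-sided material for the Constellation Sun backpack, drawing an animated emissive sun effect.
# Two scrolling noise layers, one mapped by atan2 angle and one by radial distance, combine into a signal that scales texture UVs about the center and hue-shifts the emissive color.
# The FX texture blends that glow with a radial base gradient.
# Diffuse alpha drives opacity, which a center mask and TwoSidedSign vary per face.
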Material /BRCosmetics/Accessories/FORT_Backpacks/Backpack_F_MED_Constellation_Sun/Materials/M_F_MED_ConstellationSunBackpack
graph {
 properties {
  bUseMaterialAttributes=true
  bUsedWithMeshDeformer=true
  bUsedWithSkeletalMesh=true
  blendMode=BLEND_Translucent
  pixelDepthOffsetMode=PDOM_Legacy
  twoSided=true
  usedShadingModels=Default Lit
 }
 nodes [
  n1 = Subtract(A=TextureCoordinate() {uTiling=2, vTiling=2}, B=Constant() {r=1})
  n2 = ComponentMask(None=Arctangent2Fast(Y=ComponentMask(None=@n1) {g=true}, X=ComponentMask(None=@n1) {r=true})) {r=true}
  n3 = Constant() {r=6.2832}
  n4 = Frac(None=Add(A=Divide(A=@n2, B=@n3), B=ScalarParameter() {defaultValue=0.05, parameterName=Param_3}))
  n5 = ScalarParameter() {defaultValue=10, parameterName=Param_4}
  Divide(A=Floor(None=Multiply(A=@n4, B=@n5)), B=@n5)
  n6 = SphereMask(
   A=TextureCoordinate()
   B=Constant() {r=0.5}
   Radius=ScalarParameter() {defaultValue=0.44}
   Hardness=ScalarParameter() {defaultValue=0.5238, parameterName=Param_1}
  )
  n7 = LinearInterpolate(
   A=VectorParameter() {parameterName=BaseColor, defaultValue=<rgba 0.5 0.0362 0.0511 1>}.RGB
   B=VectorParameter() {parameterName=BaseColorB, defaultValue=<rgba 0 0 0 1>}.RGB
   Alpha=@n6
  )
  n8 = AppendVector(A=Constant() {r=0.1}, B=ScalarParameter() {defaultValue=3, group=ScrollingEmissive, parameterName=Tile00})
  n9 = Constant() {r=0.33}
  n10 = Constant()
  n11 = Panner(
   Coordinate=Add(A=Multiply(A=@n4, B=@n8), B=AppendVector(A=@n9, B=@n10))
   Speed=AppendVector(A=Constant(), B=ScalarParameter() {defaultValue=1, group=ScrollingEmissive, parameterName=Speed})
  ) {speedY=1}
  n12 = SphereMask(
   A=Frac(None=@n11)
   B=Constant() {r=0.5}
   Radius=ScalarParameter() {defaultValue=0.5, group=ScrollingEmissive, parameterName=SecondaryGradientRadius00}
   Hardness=ScalarParameter() {group=ScrollingEmissive, parameterName=SecondaryGradientHardness00}
  ) {attenuationRadius=0.5, hardnessPercent=0}
  n13 = MaterialFunctionCall!(
   PrimaryTextureCoordinates=@n11
   2ndOrderTextureCoordinates
   RandomSeed
   NoiseDensityBias=ScalarParameter() {defaultValue=1, group=ScrollingEmissive, parameterName=NoiseDensityBias00}
   Bypass Noise Influence Crosstalk
  ) {materialFunction=/Game/Effects/Fort_Effects/Materials/_Functions/MF_RandomUVSample}
  n14 = SphereMask(
   A=TextureCoordinate()
   B=Constant() {r=0.5}
   Radius=ScalarParameter() {defaultValue=1, parameterName=Param00}
   Hardness=ScalarParameter() {parameterName=Param_100}
  )
  n15 = AppendVector(A=Constant() {r=0.1}, B=ScalarParameter() {defaultValue=3, group=ScrollingEmissive, parameterName=Tile})
  n16 = Constant() {r=0.33}
  n17 = Constant()
  n18 = Panner(
   Coordinate=Add(A=Multiply(A=@n14, B=@n15), B=AppendVector(A=@n16, B=@n17))
   Speed=AppendVector(A=Constant(), B=ScalarParameter() {defaultValue=1, group=ScrollingEmissive, parameterName=Speed})
  ) {speedY=1}
  n19 = SphereMask(
   A=Frac(None=@n18)
   B=Constant() {r=0.5}
   Radius=ScalarParameter() {defaultValue=0.5, group=ScrollingEmissive, parameterName=SecondaryGradientRadius}
   Hardness=ScalarParameter() {group=ScrollingEmissive, parameterName=SecondaryGradientHardness}
  ) {attenuationRadius=0.5, hardnessPercent=0}
  n20 = MaterialFunctionCall(PrimaryTextureCoordinates=@n18, NoiseDensityBias=ScalarParameter() {defaultValue=3, group=ScrollingEmissive, parameterName=NoiseDensityBias}) {materialFunction=&MF_RandomUVSample}
  n21 = Add(
   A=Multiply(A=Multiply(A=@n12, B=@n13.Noise), B=ScalarParameter() {defaultValue=2, parameterName=radialScale})
   B=Multiply(A=Multiply(A=@n19, B=@n20.Noise), B=@n14)
  )
  n22 = ScalarParameter() {defaultValue=0.015, parameterName=uvscale}
  n23 = MaterialFunctionCall(UVs=TextureCoordinate(), Texture Scale=Add(A=Multiply(A=@n21, B=@n22))) {materialFunction=&ScaleUVsByCenter}
  n24 = TextureSample(UVs=@n23.UVs) {texture=/BRCosmetics/Accessories/FORT_Backpacks/Backpack_F_MED_Constellation_Sun/Textures/T_F_MED_ConstellationSun_Backpack_Back_FX}
  n25 = OneMinus(None=@n24.R)
  n26 = Multiply(A=@n21, B=ScalarParameter() {defaultValue=-0.003, parameterName=HueShift})
  n27 = VectorParameter() {parameterName=EmissiveColor, defaultValue=<rgba 2.6 0.3326 0.234 0>}
  n28 = Add(
   A=Multiply(A=Multiply(A=@n7, B=@n25), B=ScalarParameter() {defaultValue=1, parameterName=Param_2})
   B=Multiply(
    A=@n24.R
    B=MaterialFunctionCall(Hue Shift Percentage=@n26, Texture=@n27.RGB) {materialFunction=&HueShift}.Result
   )
  )
  n29 = TextureSample(UVs=@n23.UVs) {texture=/BRCosmetics/Accessories/FORT_Backpacks/Backpack_F_MED_Constellation_Sun/Textures/T_F_MED_ConstellationSun_Backpack_Back_D}
  n30 = LinearInterpolate(A=@n29.A, B=Multiply(A=@n29.A, B=OneMinus(None=@n6)), Alpha=TwoSidedSign())
  MaterialFunctionCall(In=SetMaterialAttributes!(
    MaterialAttributes
    Emissive Color=NamedRerouteDeclaration(None=@n28) {name=emissive, nodeColor=<rgba 1 0.1641 0 1>}
    Opacity=NamedRerouteDeclaration(None=@n30) {name=opacity, nodeColor=<rgba 0 1 0.625 1>}
   ) {attributeSetTypes [#1, #2]}) {materialFunction=&MF_TechArtOverrides_accessories}
  Multiply(A=@n24.R, B=@n6)
  Saturate(None=Add(A=@n6, B=@n24.R))
  ScalarParameter() {defaultValue=1, parameterName=Specular}
  ScalarParameter() {defaultValue=0.2, parameterName=Roughness}
  TextureCoordinate()
  VectorParameter() {parameterName=SpecularColor, defaultValue=<rgba 1 0.7486 0.4062 0>}
 ]
}
