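# Lit material for the Constellation Sun backpack cosmetic that draws an animated, glowing kaleidoscope.
# Centered UVs become mirrored angular wedges and radius, offset over time, to sample a pattern texture.
# Its green channel indexes two curve-atlas color rows, blended by a custom cosine transition and circle-masked.
# Two tinted Fresnel glows are added, then an inverted sphere mask gates emissive and drives specular; gradient custom nodes are unused.
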Material /BRCosmetics/Accessories/FORT_Backpacks/Backpack_F_MED_Constellation_Sun/Materials/M_F_MED_ConstellationSunBackpack_Kaleidoscope
graph {
 properties {
  bUseMaterialAttributes=true
  bUsedWithMeshDeformer=true
  bUsedWithSkeletalMesh=true
  pixelDepthOffsetMode=PDOM_Legacy
  usedShadingModels=Default Lit
 }
 nodes [
  n1 = ScalarParameter() {defaultValue=-25, parameterName=KALEIDOSCOPE_SPEED_Y}
  AppendVector(A=ScalarParameter() {parameterName=KALEIDOSCOPE_SPEED_X}, B=@n1)
  n2 = Multiply(A=Time(), B=Constant() {r=0.3})
  MaterialExpressionAppendVector_2 = AppendVector(A=@n2, B=@n2)
  MaterialExpressionConstant_3 = Constant() {r=0.2}
  MaterialExpressionConstant3Vector_0 = Constant3Vector() {constant=<rgba 0.5 0.5 0.5 0>}
  n3 = Custom!(uv=@MaterialExpressionAppendVector_2, COLOR_TRANSITION_SIZE=@MaterialExpressionConstant_3, tex=@MaterialExpressionConstant3Vector_0) {
   code="float d = (cos(uv.y+1.2*cos(uv.x)*tex) * 0.5) + 0.5;\r\nd = smoothstep(0.5 - COLOR_TRANSITION_SIZE, 0.5 + COLOR_TRANSITION_SIZE, d);\r\nreturn d;"
   description=ColorTransition
   outputType=CMOT_Float1
   inputs [
    {
     inputName=uv
     input {expression=&self:MaterialExpressionAppendVector_2}
    }
    {
     inputName=COLOR_TRANSITION_SIZE
     input {expression=&self:MaterialExpressionConstant_3}
    }
    {
     inputName=tex
     input {expression=&self:MaterialExpressionConstant3Vector_0}
    }
   ]
  }
  n4 = Time()
  n5 = Multiply(A=Multiply(A=Add(A=@n4) {constB=1207}, B=@n1)) {constB=0.005}
  n6 = Subtract(A=Multiply(A=MaterialFunctionCall() {materialFunction=/Engine/Functions/Engine_MaterialFunctions02/Texturing/3DSandMayaUVCoordinates}.UVChannel0) {constB=2})
  n7 = MaterialFunctionCall(Float2=@n6) {materialFunction=&BreakOutFloat2Components}
  n8 = Divide(A=Arctangent2Fast(Y=@n7.G, X=@n7.G), B=Constant() {r=3.1416})
  MaterialExpressionAdd_1 = Add(A=Multiply(A=Add(A=@n8)) {constB=0.5}) {constB=0.25}
  n9 = ScalarParameter() {defaultValue=9, parameterName=KALEIDOSCOPE_SPLITS}
  MaterialExpressionDivide_1 = Divide(B=@n9)
  n10 = Custom!(x=@MaterialExpressionAdd_1, y=@MaterialExpressionDivide_1) {
   code="return x - y * floor(x/y);"
   description=mod
   outputType=CMOT_Float1
   inputs [
    {
     inputName=x
     input {expression=&self:MaterialExpressionAdd_1}
    }
    {
     inputName=y
     input {expression=&self:MaterialExpressionDivide_1}
    }
   ]
  }
  n11 = Subtract(A=Multiply(A=Multiply(A=@n10, B=@n9)) {constB=2})
  n12 = Multiply(A=@n11) {constB=-1}
  n13 = Multiply(A=Add(A=Multiply(A=@n11, B=@n12)), B=ScalarParameter() {defaultValue=1, parameterName=U_Tile}) {constB=0.5}
  n14 = Multiply(A=Distance(A=@n6, B=Constant()), B=ScalarParameter() {defaultValue=1, parameterName=V_Tile})
  n15 = Frac(None=Add(A=AppendVector(A=@n3, B=@n5), B=AppendVector(A=@n13, B=@n14)))
  MaterialExpressionTextureSampleParameter2D_0 = TextureSampleParameter2D(UVs=@n15) {parameterName=Param, texture=/BRCosmetics/Accessories/FORT_Backpacks/Backpack_F_MED_Constellation_Sun/Textures/Pattern2}
  Custom!(t=@MaterialExpressionTextureSampleParameter2D_0.RGB) {
   code="float x = (smoothstep(0.0, 0.35, t) + t*0.5);\r\nfloat y = smoothstep(0.5, 1.0, t);\r\nfloat z = max(1.0 - t*1.7, t*7.0 - 6.0);\r\n\r\nreturn clamp((pow(t, 1.5) * 0.8 + 0.2) * float3(x, y, z), 0.0, 1.0);"
   description=HeatMapGradient
   inputs [{
     inputName=t
     input {
      expression=&self:MaterialExpressionTextureSampleParameter2D_0
      mask=1
      maskB=1
      maskG=1
      maskR=1
     }
    }]
  }
  Custom(t=@MaterialExpressionTextureSampleParameter2D_0.RGB) {
   code="float tt = t*-0.9;\r\ntt = tt - 42.0 * floor(tt/42.0);\r\nreturn 0.5 + 0.5*cos( 3.0 + tt*0.075*tt + float3(0.0,0.6,1.0));"
   description=CustomGradient
   inputs [{
     inputName=t
     input {
      expression=&self:MaterialExpressionTextureSampleParameter2D_0
      mask=1
      maskB=1
      maskG=1
      maskR=1
     }
    }]
  }
  Fmod(A=@MaterialExpressionAdd_1, B=@MaterialExpressionDivide_1)
  MaterialFunctionCall(XYZ=@n15, XY=@n15) {materialFunction=&Swizzle}
  n16 = OneMinus(None=SphereMask(
    A=TextureCoordinate()
    B=Constant() {r=0.5}
    Radius=ScalarParameter() {defaultValue=0.8}
    Hardness=ScalarParameter() {defaultValue=0.2, parameterName=Param_1}
   ))
  n17 = ScalarParameter() {defaultValue=1, parameterName=Roughness}
  n18 = PixelNormalWS()
  n19 = VertexNormalWS()
  n20 = OneMinus(None=DotProduct(A=CameraVectorWS(), B=StaticSwitchParameter(True=@n18, False=@n19) {group=Constellation, parameterName=FresnelUsesPixelNormals}))
  n21 = Subtract(A=@n20, B=ScalarParameter() {defaultValue=-0.1524, group=Constellation, parameterName=CS_EmissiveFalloff}) {constB=0.35}
  n22 = ScalarParameter() {defaultValue=15.3339, group=Constellation, parameterName=CS_EmissvePower}
  n23 = Multiply(
   A=Clamp(None=Power(Base=@n21, Exp=@n22))
   B=VectorParameter() {group=Constellation, parameterName=CS_EmissiveColor, defaultValue=<rgba 2 1.7708 1.8723 1>}.RGB
  )
  n24 = Subtract(A=@n20, B=ScalarParameter() {defaultValue=-0.2286, group=Constellation, parameterName=CS_EmissiveFalloff2}) {constB=0.35}
  n25 = ScalarParameter() {defaultValue=5.7144, group=Constellation, parameterName=CS_EmissvePower2}
  n26 = Multiply(
   A=Clamp(None=Power(Base=@n24, Exp=@n25))
   B=VectorParameter() {group=Constellation, parameterName=CS_EmissiveColor2, defaultValue=<rgba 3 0.2805 0.4165 1>}.RGB
  )
  n27 = Constant() {r=0.5}
  n28 = TextureCoordinate()
  n29 = ComponentMask(None=@MaterialExpressionTextureSampleParameter2D_0.RGB) {g=true}
  n30 = CurveAtlasRowParameter(CurveTime=@n29) {
   atlas=/Game/Characters/Shared/CurveAtlas/AthenaCharacterCurveAtlas
   curve=/Game/Accessories/FORT_Backpacks/Backpack_F_MED_Constellation_Sun/Materials/Curve_ConstellationSun_Red
   defaultValue=7
   parameterName=Param_2
  }
  n31 = CurveAtlasRowParameter(CurveTime=@n29) {
   atlas=&AthenaCharacterCurveAtlas
   curve=/Game/Accessories/FORT_Backpacks/Backpack_F_MED_Constellation_Sun/Materials/Curve_ConstellationSun_Red1
   defaultValue=8
   parameterName=Param_3
  }
  MaterialExpressionReroute_1 = Reroute(None=@n15)
  MaterialExpressionScalarParameter_3 = ScalarParameter() {defaultValue=0.15, parameterName=COLOR_TRANSITION_SIZE}
  n32 = Custom!(
   uv=@MaterialExpressionReroute_1
   COLOR_TRANSITION_SIZE=@MaterialExpressionScalarParameter_3
   tex=@MaterialExpressionTextureSampleParameter2D_0.RGB
  ) {
   code="float d = (cos(uv.y+1.2*cos(uv.x)*tex) * 0.5) + 0.5;\r\nd = smoothstep(0.5 - COLOR_TRANSITION_SIZE, 0.5 + COLOR_TRANSITION_SIZE, d);\r\nreturn d;"
   description=ColorTransition
   outputType=CMOT_Float1
   inputs [
    {
     inputName=uv
     input {expression=&self:MaterialExpressionReroute_1}
    }
    {
     inputName=COLOR_TRANSITION_SIZE
     input {expression=&self:MaterialExpressionScalarParameter_3}
    }
    {
     inputName=tex
     input {
      expression=&self:MaterialExpressionTextureSampleParameter2D_0
      mask=1
      maskB=1
      maskG=1
      maskR=1
     }
    }
   ]
  }
  n33 = Multiply(
   A=MaterialFunctionCall(Alpha=Distance(A=@n27, B=@n28), Min=Constant() {r=0.4695}, Max=Constant() {r=0.3262}) {materialFunction=&SmoothStep}.Result
   B=LinearInterpolate(A=@n30, B=@n31, Alpha=@n32)
  )
  n34 = ScalarParameter() {defaultValue=1.5, parameterName=EmissiveBoost}
  n35 = Multiply(A=@n16, B=Add(A=Add(A=@n23, B=@n26), B=Multiply(A=@n33, B=@n34)))
  MaterialFunctionCall(In=SetMaterialAttributes!(
    MaterialAttributes
    Specular=NamedRerouteDeclaration(None=@n16) {name=Specular, nodeColor=<rgba 0 1 0.9063 1>}
    Roughness=NamedRerouteDeclaration(None=@n17) {name=Roughness, nodeColor=<rgba 1 0.3047 0 1>}
    Emissive Color=NamedRerouteDeclaration(None=@n35) {name=Emissive, nodeColor=<rgba 0 0.6016 1 1>}
   ) {attributeSetTypes [#1, #2, #3]}) {materialFunction=&MF_TechArtOverrides_accessories}
 ]
}
